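# Parent material for the Gold skin of the FutureRabbit pickaxe, built on the 2023 character parent PreFX and PostFX functions.
# Between them, material attributes chain through switchable hair, Ice, Galaxy, carbon fiber, holographic glitter, DoubleConk grid and FutureRabbit effect functions.
# An isGold static switch enables the carbon fiber layer and blends its normal by mask into the separate clear coat normal output.
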Material /BRCosmetics/Weapons/FORT_Melee/Pickaxe_FutureRabbit/Skins/Gold/Materials/M_Pickaxe_FutureRabbit_Gold_Parent
graph {
 properties {
  bUseMaterialAttributes=true
  bUsedWithMeshDeformer=true
  bUsedWithMorphTargets=true
  bUsedWithSkeletalMesh=true
  pixelDepthOffsetMode=PDOM_Legacy
  previewMesh=/BRCosmetics/Weapons/FORT_Melee/Pickaxe_FutureRabbit/Meshes/FutureRabbit_Axe
  shadingModel=MSM_FromMaterialExpression
  subsurfaceProfile=/Game/Packages/Fortress_SharedMaterials/SubSurfaceProfile/Character_SubsurfaceProfile
  usedShadingModels=Default Lit | Subsurface | Clear Coat | Subsurface Profile | Cloth
 }
 nodes [
  n1 = MaterialFunctionCall() {materialFunction=&MF_CharacterParent_2023_PreFX}
  n2 = MaterialFunctionCall!(In=@n1.Result, Mask Input=NamedRerouteUsage()) {materialFunction=/Game/Global/Materials/Character_Parent_2023/Hair_Parent_2023/MaterialFunctions/MF_FN_Hair_2023}
  n3 = StaticSwitchParameter(True=BlendMaterialAttributes(A=@n1.Result, B=@n2.Result, Alpha=@n2.Result), False=@n1.Result) {group=Skin FX, parameterName=UseFNHair2_0}
  n4 = MaterialFunctionCall!(MaterialInput=@n3) {materialFunction=/Game/Global/Materials/Character_Parent_2023/Hair_Parent_2023/MaterialFunctions/MF_FlyAwayHair_2023}
  n5 = MaterialFunctionCall!(
   Tech Art Mask=NamedRerouteUsage()
   Mask=NamedRerouteUsage()
   In=MaterialFunctionCall!(
    TechArt=NamedRerouteUsage()
    Fresnel=NamedRerouteUsage()
    In=StaticSwitchParameter(True=@n4.Result, False=@n3) {group=Skin FX, parameterName=IsFlyAwayHair?}
   ) {materialFunction=/Game/Global/Materials/Character_Parent_2023/CrewStyle_Parent_2023/Ice/MF_CrewStyle_Ice_Inst}.Result
  ) {materialFunction=/Game/Global/Materials/Character_Parent_2023/CrewStyle_Parent_2023/Galaxy/MF_CrewStyle_Galaxy_Inst}
  n6 = MaterialFunctionCall!(Alpha=NamedRerouteUsage(), In=@n5.Result) {materialFunction=/Game/Global/MaterialFunctions/TechArt/MF_CarbonFiberDualLayer}
  n7 = GetMaterialAttributes(None=@n6.Result) {attributeGetTypes [#1]}
  n8 = MaterialFunctionCall(
   BaseNormal=@n7.Normal
   AdditionalNormal=TextureSample(UVs=@n6.CarbonFiberUvs) {samplerType=SAMPLERTYPE_Normal, texture=/Game/Weapons/WeaponSkins/Wraps/Textures/T_Wrap_CarbonFiber_N}.RGB
  ) {materialFunction=&BlendAngleCorrectedNormals}
  ClearCoatNormalCustomOutput!(None=StaticSwitchParameter(True=LinearInterpolate(A=@n7.Normal, B=@n8.Result, Alpha=@n6.Mask), False=@n7.Normal) {parameterName=isGold})
  n9 = MaterialFunctionCall!(Alpha=NamedRerouteUsage(), In=StaticSwitchParameter(True=@n6.Result, False=@n5.Result) {parameterName=isGold}, Fresnel=NamedRerouteUsage()) {materialFunction=/Game/Global/MaterialFunctions/TechArt/MF_GlitterHolographic}
  n10 = NamedRerouteDeclaration(None=@n1.Tech Art Mask) {name=TechArtMask, nodeColor=<rgba 1 0.6797 0 1>}
  MaterialFunctionCall(
   Material=MaterialFunctionCall!(
    Material=MaterialFunctionCall!(MatIn=@n9.Result, TechArtMask=@n10) {materialFunction=/BRCosmetics/Characters/Player/Male/Medium/Bodies/M_MED_DoubleConk/Materials/MF_DoubleConkGrid}.Result
    Tech Art Mask=NamedRerouteUsage()
   ) {materialFunction=/BRCosmetics/Weapons/FORT_Melee/Pickaxe_FutureRabbit/Materials/MF_Pickaxe_FutureRabbit}.Result
   M=@n1.M
   Pre Skinned Local Position=@n1.M
   Distance Blend=@n1.M
   Fresnel=@n1.M
  ) {materialFunction=&MF_CharacterParent_2023_PostFX}
  NamedRerouteDeclaration(None=@n1.M) {name=Mask, nodeColor=<rgba 0 1 1 1>}
  NamedRerouteDeclaration(None=@n1.Fresnel) {name=Fresnel, nodeColor=<rgba 0 1 0.4375 1>}
 ]
}
